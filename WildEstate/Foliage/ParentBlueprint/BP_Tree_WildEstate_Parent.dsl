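; Parent Blueprint for Wild Estate trees that adds leaf-colored destruction effects on death.
; On death it runs the parent logic, then on mobile only spawns a toon tree-destruction Niagara system and applies FX customization to it.
; The customization callback sets that effect's LeafBaseColor from the tree's LeaveColor, with debug prints along the way.
; The construction script makes the static mesh affect distance field lighting.
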(event EventBeginPlay
  (|CallParentFunction) ; This node is disabled and will not be called.\nDrag off pins to build functionality.
)

(event Collision|EventActorBeginOverlap
  (params (OtherActor :type Actor%20Object%20Reference)))

(event EventTick
  (params (DeltaSeconds :type Float%20%28single-precision%29)))

(event EventOnDeathPlayEffects
  (params (Damage :type Float%20%28single-precision%29) (DamageTags :type Gameplay%20Tag%20Container%20Structure%20%28by%20ref%29) (Momentum :type Vector :default "0, 0, 0") (HitInfo :type Hit%20Result%20Structure%20%28by%20ref%29) (InstigatedBy :type Fort%20Pawn%20Object%20Reference) (DamageCauser :type Actor%20Object%20Reference) (EffectContext :type Gameplay%20Effect%20Context%20Handle%20Structure))
  (bind (_returnvalue_location _returnvalue_rotation) (Math|Transform|BreakTransform (Transformation|GetActorTransform)))
  (|CallParentFunction Damage DamageTags Momentum HitInfo InstigatedBy DamageCauser EffectContext)
  (if (TouchPlatform|IsMobileGame)
    (bind _returnvalue (Niagara|SpawnSystematLocation "/Game/Effects/Fort_Effects/World/Destruction/Destruction2/Trees/NS_Destruction_Toon_Trees.NS_Destruction_Toon_Trees" _returnvalue_location _returnvalue_rotation))
    (bind _returnvalue_2 (Class|FortFXCustomizationComponent|ApplyFXCustomizationtoComponent self _returnvalue))
    (Development|PrintString (Utilities|String|Append "Customization Applied:" (Utilities|String|ToString(Boolean) _returnvalue_2)))))

(event OnFXCustomizationApplied(FortFXCustomization)
  (params (FXCustomizationComp :type Fort%20FXCustomization%20Component%20Object%20Reference) (TargetFXComp :type FXSystem%20Component%20Object%20Reference) (TargetFXAsset :type FXSystem%20Asset%20Object%20Reference))
  (bind (_r _g _b) (Utilities|Struct|BreakLinearColor (Variables|LeaveColor|GetColor)))
  (Development|PrintString (Utilities|String|Append "FXCustomCOmp: " (Utilities|GetDisplayName FXCustomizationComp) "
" "Target FXComp: " (Utilities|GetDisplayName TargetFXComp) "
" "Target FXAsset: " (Utilities|GetDisplayName TargetFXAsset)) true true "(R=0.000000,G=0.660000,B=1.000000,A=1.000000)" 2e+01)
  (Effects|Components|ParticleSystem|SetColorParameter TargetFXComp "LeafBaseColor" (Utilities|Struct|MakeLinearColor _r _g _b 1.0)))

(fn ConstructionScript
  (|CallParentFunction)
  (Rendering|SetAffectDistanceFieldLighting (Variables|Debug|GetStaticMeshComponent) true))
